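# Sound cue for the frag grenade cook sound, routed to the gadget items sound class.
# A random node picks one of two equally weighted cook waves, then a modulator varies its pitch and volume.
# It plays at reduced volume with the default grenade cook attenuation.
# A concurrency override allows one instance per owner, stopping the oldest when a new one starts.
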SoundCue /Game/Sounds/Fort_Weapons/Grenades/Frag/Grenade_Frag_Cook_Cue
root=Modulator_0
nodes [
 {id=Modulator_0, children [Random_0], properties {pitchMax=1.1, volumeMin=0.7}}
 {id=Random_0, children [WavePlayer_0,WavePlayer_1], properties {bShouldExcludeFromBranchCulling=true, weights [1,1]}}
 {id=WavePlayer_0, properties {soundWaveAssetPtr=&/Grenade_Frag_Cook_01}}
 {id=WavePlayer_1, properties {soundWaveAssetPtr=&/Grenade_Frag_Cook_02}}
]
properties {
 attenuationSettings=/Game/Sounds/Fort_Weapons/Grenades/Grenade_Default_Cook_Att
 bOverrideConcurrency=true
 maxDistance=3000
 soundClassObject=&Fort_Gadgets_Items
 volumeMultiplier=0.2
 concurrencyOverrides {bLimitToOwner=true, maxCount=1, resolutionRule=StopOldest}
}
