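# Sound cue for the drop sound of the frag grenade.
# A single wave player feeds a modulator that slightly randomizes pitch while keeping volume fixed.
# It plays at reduced volume under the loot collect sound class with weapon drop attenuation.
# A pickup concurrency set is assigned alongside overrides that allow one instance at a time.
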SoundCue /Game/Sounds/Fort_Weapons/Grenades/Frag/Grenade_Frag_Drop_Cue
root=Modulator_1
nodes [
 {id=Modulator_1, children [WavePlayer_5], properties {pitchMax=1.1, pitchMin=0.9, volumeMax=1, volumeMin=1}}
 {id=WavePlayer_5, properties {soundWaveAssetPtr=&/Grenade_Frag_Drop_01}}
]
properties {
 attenuationSettings=&Fort_Weapon_Drop_Att
 maxDistance=3000
 soundClassObject=&Fort_Loot_Collect
 volumeMultiplier=0.35
 concurrencyOverrides {maxCount=1}
 concurrencySet [&PickupMaximum_Oldest]
}
