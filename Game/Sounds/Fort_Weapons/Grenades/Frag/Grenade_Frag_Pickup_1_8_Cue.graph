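# Sound cue named for the frag grenade pickup, playing a teleporter gadget pickup wave.
# A single wave player feeds the output directly at reduced volume, with no randomization or layering.
# It uses the shared weapon pickup attenuation, the loot collect sound class and a pickup concurrency set.
# A concurrency override with a prevent-new rule is also stored.
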SoundCue /Game/Sounds/Fort_Weapons/Grenades/Frag/Grenade_Frag_Pickup_1_8_Cue
root=WavePlayer_0
nodes [{id=WavePlayer_0, properties {soundWaveAssetPtr=/Game/Sounds/Fort_Gadgets/Teleporter/teleporter_pickup_01}}]
properties {
 attenuationSettings=&Fort_Weapon_Pickup_Att
 maxDistance=3500
 soundClassObject=&Fort_Loot_Collect
 volumeMultiplier=0.5
 concurrencyOverrides {resolutionRule=PreventNew}
 concurrencySet [&PickupMaximum_Oldest]
}
